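# Niagara system for the GlitchyBit vehicle boost cosmetic, drawing glitchy ribbon trails and SubUV sprite bits from the car's skeletal-mesh sockets.
# Blue-outline and green ribbon emitters run per port, with spawn groups selecting opposite-signed widths for multiple renderers.
# Beginning, final and hole sprite emitters partition particles across sockets, tag renderers by spawn group and drive layered colors from timed alpha curves.
# GameModeID switches tune lifetimes and ribbon speed; painted-variant overrides are disabled.
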NiagaraSystem /VehicleCosmetics/Boosts/Boost_GlitchyBit/FX/NS_GlitchyBit
dependencies {
 usedDataInterfaces [
  &NiagaraDataInterfaceColorCurve
  &NiagaraDataInterfaceArrayFloat
  &NiagaraDataInterfaceSpriteRendererInfo
  &NiagaraDataInterfaceVector2DCurve
  &NiagaraDataInterfaceSkeletalMesh
 ]
 usedDynamicInputs []
 usedModules [
  &self:SystemSpawnScript.SystemScriptSource.SystemScriptGraph.NiagaraNodeAssignment_0.SetVariables_#1
  &SystemState
 ]
}
emitters {
 NE_BegginingParticles {
  emitterUpdateScript [
   &EmitterState()
   &SpawnRate(SpawnGroup=1, SpawnRate=80, Use Spawn Probability=false) {module=SpawnRate001}
   &VC_FilteredSocketCount()
   &SpawnRate(SpawnRate=75, Use Spawn Probability=false)
  ]
  particleSpawnScript [
   &VC_PaintedVariantsOverride() {enabled=false}
   &InitializeParticle(
    Color Mode=::NewEnumerator0
    Lifetime=&MergeHelper_FloatSwitch(0=0.3, 1=0.3, 2=0.3, 3=?, Index=$User.GameModeID, Num=::NewEnumerator1)
    Position Mode=::NewEnumerator0
    Sprite Rotation Mode=::NewEnumerator2
    Sprite Size Max=<20 20>
    Sprite Size Min=<10 10>
    Sprite Size Mode=::NewEnumerator2
    Write Material Random=true
   )
   &PartitionParticles(Number Of Partitions=$Emitter.PortCount, Partition Mode:ENiagara_PartitionMode=::NewEnumerator5)
   &SkeletalMeshLocation(
    Bone / Socket Index=$Particles.PartitionParticles.Partition
    Bone / Socket Sampling Mode=::NewEnumerator5
    MaskSpawnGroup=false
    Mesh Sampling Type=::NewEnumerator6
    Sampled Position Offset=&AddVectorToPosition(Position=<0 0 0>, Vector=<-130 0 0>)
    Skeletal Mesh=$System.SkeletalMesh
   )
   &AddVelocity(Velocity=&MakeCustomVectorFromBool(
     Bool=$User.bIsFrontendPreview
     True Vector=&RandomRangeVector(Minimum=<200 0 0>, Maximum=<300 0 0>)
     False Vector=&RandomRangeVector(Minimum=<200 0 0>, Maximum=<300 0 0>)
    ))
   &ShapeLocation(MaskSpawnGroup=false, Override Seed=false, Sphere Radius=30, UseOwnerScale=true)
   &SubUVAnimation(
    "Sprite Renderer "=DataInterface() {propertyValues {SpriteRenderer=&self:NE_BegginingParticles_0.NiagaraSpriteRendererProperties_0}}
    SubUV Animation Mode:ENiagara_SubUVLookupModeV2=::NewEnumerator2
    UseEndFrame=false
    UseStartFrame=false
   )
   /SetVariables(Particles.RendererTag:NiagaraInt32=&SpawnGroup()) {isSetParameters=true, module=SetVariables_#2}
  ]
  particleUpdateScript [
   &ParticleState()
   &InheritSourceMovement(
    Applied Position Scale=&Multiply_VectorByFloat(Vector=<1 1 1>, Float=1)
    ApplyForce=false
    ApplyPosition=true
    ApplyPositionFalloff=false
   )
   &SolveForcesAndVelocity(Clamp Velocity=false, Limit Acceleration=false)
   /SetVariables(
    Particles.GreenInsideColor:LinearColor=&ColorFromCurve(ColorCurve=DataInterface() {propertyValues {
       AlphaCurve {keys [<at 0.4117 0>,<at 0.4399 1>,<at 0.6429 1>,<at 0.6578 0>]}
       BlueCurve {keys [<at 0.0783 0.0464>]}
       GreenCurve {keys [<at 0.0783 0.4896>]}
       RedCurve {keys [<at 0.0783 0.3031>]}
      }})
    Particles.GreenOutline_Color:LinearColor=&ColorFromCurve(ColorCurve=DataInterface() {propertyValues {
       AlphaCurve {keys [<at 0.4117 0 RCIM_Cubic>,<at 0.4413 1 RCIM_Cubic>,<at 0.6433 1>,<at 0.6588 0>]}
       BlueCurve {keys [<at 0 0.0143 RCIM_Cubic>]}
       GreenCurve {keys [<at 0 0.6875 RCIM_Cubic>]}
       RedCurve {keys [<at 0 0.5556 RCIM_Cubic>]}
      }})
    Particles.PinkOffsetColor:LinearColor=&ColorFromCurve(ColorCurve=DataInterface() {propertyValues {
       AlphaCurve {keys [<at 0.3193 0>,<at 0.3775 1>,<at 0.6857 0.9369>,<at 0.7336 0>]}
       BlueCurve {keys [<at 0.0684 0.4678>]}
       GreenCurve {keys [<at 0.0684 0.0176>]}
       RedCurve {keys [<at 0.0684 0.003>]}
      }})
    Particles.PurpleColor:LinearColor=&ColorFromCurve(ColorCurve=DataInterface() {propertyValues {
       AlphaCurve {keys [<at 0.3193 0>,<at 0.3946 1>,<at 0.6583 1>,<at 0.6788 0>]}
       BlueCurve {keys [<at 0.0684 0.8333>]}
       GreenCurve {keys [<at 0.0684 0.486>]}
       RedCurve {keys [<at 0.0684 0.013>]}
      }})
    Particles.SpriteSize=&Vector2DFromCurve(
     Vector2Curve=DataInterface() {propertyValues {XCurve {keys [<at 0.5063 1>,<at 1 0>]}, YCurve {keys [<at 0.5063 1>,<at 1 0>]}}}
     ScaleCurve=<25 25>
    )
    Particles.YellowSize:Vector2f=&RandomRangeVector2D(Minimum=<10 10>, Maximum=<22 22>)
   ) {isSetParameters=true, module=SetVariables_#3}
  ]
  rendererClasses [&NiagaraSpriteRendererProperties]
  renderers [
   {rendererClass=&NiagaraSpriteRendererProperties}
   {rendererClass=&NiagaraSpriteRendererProperties, rendererIndex=1}
   {rendererClass=&NiagaraSpriteRendererProperties, rendererIndex=2}
   {rendererClass=&NiagaraSpriteRendererProperties, rendererIndex=3}
  ]
 }
 NE_Final_Particles {
  emitterUpdateScript [
   &EmitterState()
   &SpawnRate(SpawnGroup=1, SpawnRate=20, Use Spawn Probability=false) {module=SpawnRate001}
   &VC_FilteredSocketCount()
   &SpawnRate(SpawnRate=25, Use Spawn Probability=false)
  ]
  particleSpawnScript [
   &VC_PaintedVariantsOverride() {enabled=false}
   &InitializeParticle(
    Color Mode=::NewEnumerator0
    Lifetime=&MergeHelper_FloatSwitch(0=0.7, 1=0.65, 2=0.65, 3=?, Index=$User.GameModeID, Num=::NewEnumerator1)
    Position Mode=::NewEnumerator0
    Sprite Rotation Mode=::NewEnumerator2
    Sprite Size Max=<25 25>
    Sprite Size Min=<15 15>
    Sprite Size Mode=::NewEnumerator2
    Write Material Random=true
   )
   &PartitionParticles(Number Of Partitions=$Emitter.PortCount)
   &SkeletalMeshLocation(
    Bone / Socket Index=$Particles.PartitionParticles.Partition
    Bone / Socket Sampling Mode=::NewEnumerator5
    MaskSpawnGroup=false
    Mesh Sampling Type=::NewEnumerator6
    Sampled Position Offset=&AddVectorToPosition(Position=<0 0 0>, Vector=<-50 0 0>)
    Skeletal Mesh=$System.SkeletalMesh
   )
   &AddVelocity(Velocity=&MakeCustomVectorFromBool(
     Bool=$User.bIsFrontendPreview
     True Vector=&RandomRangeVector(Minimum=<-900 100 0>, Maximum=<-1150 -100 50>)
     False Vector=&RandomRangeVector(Minimum=<-900 100 0>, Maximum=<-1150 -100 50>)
    ))
   &ShapeLocation(MaskSpawnGroup=false, Override Seed=false, Sphere Radius=20, UseOwnerScale=true)
   &SubUVAnimation(
    "Sprite Renderer "=DataInterface() {propertyValues {SpriteRenderer=&self:NE_Final_Particles_0.NiagaraSpriteRendererProperties_0}}
    SubUV Animation Mode:ENiagara_SubUVLookupModeV2=::NewEnumerator2
    UseEndFrame=false
    UseStartFrame=false
   )
   /SetVariables(Particles.RendererTag:NiagaraInt32=&SpawnGroup()) {isSetParameters=true, module=SetVariables_#2}
  ]
  particleUpdateScript [
   &ParticleState()
   &InheritSourceMovement(
    Applied Position Scale=&Multiply_VectorByFloat(Vector=<1 1 1>, Float=1)
    ApplyForce=false
    ApplyPosition=true
    ApplyPositionFalloff=false
   )
   &SolveForcesAndVelocity(Clamp Velocity=false, Limit Acceleration=false)
   /SetVariables(
    Particles.GreenInsideColor:LinearColor=&ColorFromCurve(ColorCurve=DataInterface() {propertyValues {
       AlphaCurve {keys [<at 0.4117 0>,<at 0.4399 1>,<at 0.6429 1>,<at 0.6578 0>]}
       BlueCurve {keys [<at 0.0783 0.0464>]}
       GreenCurve {keys [<at 0.0783 0.4896>]}
       RedCurve {keys [<at 0.0783 0.3031>]}
      }})
    Particles.GreenOutline_Color:LinearColor=&ColorFromCurve(ColorCurve=DataInterface() {propertyValues {
       AlphaCurve {keys [<at 0.4117 0 RCIM_Cubic>,<at 0.4413 1 RCIM_Cubic>,<at 0.6433 1>,<at 0.6588 0>]}
       BlueCurve {keys [<at 0 0.0143 RCIM_Cubic>]}
       GreenCurve {keys [<at 0 0.6875 RCIM_Cubic>]}
       RedCurve {keys [<at 0 0.5556 RCIM_Cubic>]}
      }})
    Particles.PinkOffsetColor:LinearColor=&ColorFromCurve(ColorCurve=DataInterface() {propertyValues {
       AlphaCurve {keys [<at 0.3193 0>,<at 0.3775 1>,<at 0.6857 0.9369>,<at 0.7336 0>]}
       BlueCurve {keys [<at 0.0684 0.4678>]}
       GreenCurve {keys [<at 0.0684 0.0176>]}
       RedCurve {keys [<at 0.0684 0.003>]}
      }})
    Particles.PurpleColor:LinearColor=&ColorFromCurve(ColorCurve=DataInterface() {propertyValues {
       AlphaCurve {keys [<at 0.3193 0>,<at 0.3946 1>,<at 0.6583 1>,<at 0.6788 0>]}
       BlueCurve {keys [<at 0.0684 0.8333>]}
       GreenCurve {keys [<at 0.0684 0.486>]}
       RedCurve {keys [<at 0.0684 0.013>]}
      }})
    Particles.YellowSize:Vector2f=&RandomRangeVector2D(Minimum=<10 10>, Maximum=<22 22>)
   ) {isSetParameters=true, module=SetVariables_#3}
  ]
  rendererClasses [&NiagaraSpriteRendererProperties]
  renderers [
   {rendererClass=&NiagaraSpriteRendererProperties}
   {rendererClass=&NiagaraSpriteRendererProperties, rendererIndex=1}
   {rendererClass=&NiagaraSpriteRendererProperties, rendererIndex=2}
   {rendererClass=&NiagaraSpriteRendererProperties, rendererIndex=3}
  ]
 }
 NE_GlitchyBlueOutline {
  emitterUpdateScript [
   &EmitterState()
   &VC_FilteredSocketCount()
   &VC_RateQualityHandler(High Spawn Rate=60, Low Spawn Rate=60, Medium Spawn Rate=60, Use Custom Rate?=true)
   &VC_SpawnRibbon_v1(Number of Ribbons=$Emitter.PortCount, SpawnRate=$Emitter.Spawn Rate)
  ]
  particleSpawnScript [
   &VC_PaintedVariantsOverride() {enabled=false}
   &VC_InitializeRibbon_v1(
    Color=&MergeHelper_ColorSwitch(
     0=<rgba 0.0004 0.0057 0.2344 1>
     1=<rgba 0.0006 0.0086 0.3419 1>
     2=<rgba 0.0006 0.0086 0.3419 1>
     3=?
     Index=$System.GameModeID
     Num=::NewEnumerator1
    )
    Custom Offset=true
    Enable Skeletal Position=true
    Flip Facing=true
    Lifetime=&MergeHelper_FloatSwitch(0=0.9, 1=0.8, 2=0.8, 3=?, Index=$System.GameModeID, Num=::NewEnumerator1)
    Local Vector Offset=&AddVectorToPosition(Position=<-70 0 0>, Vector=$User.DTCarOffset)
    Offset Amount=1
    Ribbon Width=45
   )
   &AddVelocity(Velocity=&MergeHelper_VectorSwitch(0=<-1200 0 0>, 1=<-900 0 0>, 2=<-900 0 0>, 3=?, Index=$System.GameModeID, Num=::NewEnumerator1)) {module=AddVelocity001}
   /SetVariables(
    Particles.GreenR_Color:LinearColor=&ColorFromCurve(ColorCurve=DataInterface() {propertyValues {
       AlphaCurve {keys [<at 0 1 RCIM_Cubic>]}
       BlueCurve {keys [<at 0 0.0742 RCIM_Cubic>]}
       GreenCurve {keys [<at 0 0.807 RCIM_Cubic>]}
       RedCurve {keys [<at 0 0.5089 RCIM_Cubic>]}
      }})
    Particles.GreenWidth=45
    Particles.OutlineBlueColor:LinearColor=<rgba 0.0331 0.6514 0.9047 1>
   ) {isSetParameters=true, module=SetVariables_#4}
   /SetVariables(
    Particles.R1_Width:NiagaraFloat=&SelectFloatFromArray(Direct Array Index=&SpawnGroup(), Float Selection Array=DataInterface() {propertyValues {FloatData [1,0]}})
    Particles.R2_Width:NiagaraFloat=&SelectFloatFromArray(Direct Array Index=&SpawnGroup(), Float Selection Array=DataInterface() {propertyValues {FloatData [0,-1]}})
   ) {isSetParameters=true, module=SetVariables_#5}
  ]
  particleUpdateScript [
   &ParticleState()
   &InheritSourceMovement(
    Applied Position Scale=&Multiply_VectorByFloat(Vector=<1 1 1>, Float=1)
    ApplyForce=false
    ApplyPosition=true
    ApplyPositionFalloff=false
   )
   &SolveForcesAndVelocity(Clamp Velocity=false, Limit Acceleration=false)
   &VC_UpdateRibbon_v1(Enable Skeletal Position=true)
   /SetVariables(
    Particles.R1_Width:NiagaraFloat=&Multiply_Float(A=$Particles.Initial.R1_Width, B=45)
    Particles.R2_Width:NiagaraFloat=&Multiply_Float(A=$Particles.Initial.R2_Width, B=45)
   ) {isSetParameters=true, module=SetVariables_#6}
  ]
  rendererClasses [&NiagaraRibbonRendererProperties]
  renderers [
   {rendererClass=&NiagaraRibbonRendererProperties}
   {rendererClass=&NiagaraRibbonRendererProperties, rendererIndex=1}
   {rendererClass=&NiagaraRibbonRendererProperties, rendererIndex=2}
   {rendererClass=&NiagaraRibbonRendererProperties, rendererIndex=3}
  ]
 }
 NE_GlitchyGreen {
  emitterUpdateScript [
   &EmitterState()
   &VC_FilteredSocketCount()
   &VC_RateQualityHandler(High Spawn Rate=60, Low Spawn Rate=60, Medium Spawn Rate=60, Use Custom Rate?=true)
   &VC_SpawnRibbon_v1(Number of Ribbons=$Emitter.PortCount, SpawnRate=$Emitter.Spawn Rate)
  ]
  particleSpawnScript [
   &VC_PaintedVariantsOverride() {enabled=false}
   &VC_InitializeRibbon_v1(
    Color=<rgba 0.0012 0.0168 0.6939 1>
    Custom Offset=true
    Enable Skeletal Position=true
    Flip Facing=true
    Lifetime=&MergeHelper_FloatSwitch(0=0.9, 1=0.8, 2=0.8, 3=?, Index=$System.GameModeID, Num=::NewEnumerator1)
    Local Vector Offset=&AddVectorToPosition(Position=<-70 0 0>, Vector=$User.DTCarOffset)
    Offset Amount=1
    Ribbon Width=45
   )
   &AddVelocity(Velocity=&MergeHelper_VectorSwitch(0=<-1200 0 0>, 1=<-900 0 0>, 2=<-900 0 0>, 3=?, Index=$System.GameModeID, Num=::NewEnumerator1)) {module=AddVelocity001}
   /SetVariables(
    Particles.GreenR_Color:LinearColor=&ColorFromCurve(ColorCurve=DataInterface() {propertyValues {
       AlphaCurve {keys [<at 0 1 RCIM_Cubic>]}
       BlueCurve {keys [<at 0 0.0522 RCIM_Cubic>]}
       GreenCurve {keys [<at 0 0.5677 RCIM_Cubic>]}
       RedCurve {keys [<at 0 0.358 RCIM_Cubic>]}
      }})
    Particles.GreenWidth=45
    Particles.OutlineBlueColor:LinearColor=<rgba 0.0331 0.6514 0.9047 1>
   ) {isSetParameters=true, module=SetVariables_#4}
   /SetVariables(
    Particles.R1_Width:NiagaraFloat=&SelectFloatFromArray(Direct Array Index=&SpawnGroup(), Float Selection Array=DataInterface() {propertyValues {FloatData [1,0]}})
    Particles.R2_Width:NiagaraFloat=&SelectFloatFromArray(Direct Array Index=&SpawnGroup(), Float Selection Array=DataInterface() {propertyValues {FloatData [0,-1]}})
   ) {isSetParameters=true, module=SetVariables_#5}
  ]
  particleUpdateScript [
   &ParticleState()
   &InheritSourceMovement(
    Applied Position Scale=&Multiply_VectorByFloat(Vector=<1 1 1>, Float=1)
    ApplyForce=false
    ApplyPosition=true
    ApplyPositionFalloff=false
   )
   &SolveForcesAndVelocity(Clamp Velocity=false, Limit Acceleration=false)
   &VC_UpdateRibbon_v1(Enable Skeletal Position=true)
   /SetVariables(Particles.GreenOffsetPosition:NiagaraPosition=&AddVectorToPosition(
     Position=$Particles.Position
     Vector=&MaskVector(Mask Source=0, Mask Parameter=?, Mask=0, Masked Vector=<0 8 4>, Passthrough Vector=<0 -8 4>)
    )) {isSetParameters=true}
   /SetVariables(
    Particles.R1_Width:NiagaraFloat=&Multiply_Float(A=$Particles.Initial.R1_Width, B=45)
    Particles.R2_Width:NiagaraFloat=&Multiply_Float(A=$Particles.Initial.R2_Width, B=45)
   ) {isSetParameters=true, module=SetVariables_#6}
  ]
  rendererClasses [&NiagaraRibbonRendererProperties]
  renderers [
   {rendererClass=&NiagaraRibbonRendererProperties}
   {rendererClass=&NiagaraRibbonRendererProperties, rendererIndex=1}
   {rendererClass=&NiagaraRibbonRendererProperties, rendererIndex=2}
  ]
 }
 NE_HolesParticles {
  emitterUpdateScript [
   &EmitterState()
   &SpawnRate(SpawnGroup=1, SpawnRate=10, Use Spawn Probability=false) {module=SpawnRate001}
   &VC_FilteredSocketCount()
   &SpawnRate(SpawnRate=10, Use Spawn Probability=false)
  ]
  particleSpawnScript [
   &VC_PaintedVariantsOverride() {enabled=false}
   &InitializeParticle(
    Color Mode=::NewEnumerator0
    Lifetime=&MergeHelper_FloatSwitch(0=0.5, 1=0.5, 2=0.5, 3=?, Index=$User.GameModeID, Num=::NewEnumerator1)
    Position Mode=::NewEnumerator0
    Sprite Rotation Mode=::NewEnumerator2
    Sprite Size Max=<22 22>
    Sprite Size Min=<10 10>
    Sprite Size Mode=::NewEnumerator2
    Write Material Random=true
   )
   &PartitionParticles(Number Of Partitions=$Emitter.PortCount, Partition Mode:ENiagara_PartitionMode=::NewEnumerator5)
   &SkeletalMeshLocation(
    Bone / Socket Index=$Particles.PartitionParticles.Partition
    Bone / Socket Sampling Mode=::NewEnumerator5
    MaskSpawnGroup=false
    Mesh Sampling Type=::NewEnumerator6
    Sampled Position Offset=&AddVectorToPosition(Position=<0 0 0>, Vector=<0 0 0>)
    Skeletal Mesh=$System.SkeletalMesh
   )
   &AddVelocity(Velocity=&MakeCustomVectorFromBool(
     Bool=$User.bIsFrontendPreview
     True Vector=&RandomRangeVector(Minimum=<-900 0 0>, Maximum=<-1150 0 50>)
     False Vector=&RandomRangeVector(Minimum=<-900 0 0>, Maximum=<-1150 0 50>)
    ))
   &ShapeLocation(MaskSpawnGroup=false, Override Seed=false, Sphere Radius=8, UseOwnerScale=true)
   &SubUVAnimation(
    "Sprite Renderer "=DataInterface() {propertyValues {SpriteRenderer=&self:NE_HolesParticles_0.NiagaraSpriteRendererProperties_0}}
    SubUV Animation Mode:ENiagara_SubUVLookupModeV2=::NewEnumerator2
    UseEndFrame=false
    UseStartFrame=false
   )
   /SetVariables(Particles.RendererTag:NiagaraInt32=&SpawnGroup()) {isSetParameters=true, module=SetVariables_#2}
  ]
  particleUpdateScript [
   &ParticleState()
   &InheritSourceMovement(
    Applied Position Scale=&Multiply_VectorByFloat(Vector=<1 1 1>, Float=1)
    ApplyForce=false
    ApplyPosition=true
    ApplyPositionFalloff=false
   )
   &SolveForcesAndVelocity(Clamp Velocity=false, Limit Acceleration=false)
   /SetVariables(
    Particles.GreenInsideColor:LinearColor=&ColorFromCurve(ColorCurve=DataInterface() {propertyValues {
       AlphaCurve {keys [<at 0.2368 0>,<at 0.2908 1>,<at 0.6429 1>,<at 0.6578 0>]}
       BlueCurve {keys [<at 0.0783 0.0185>]}
       GreenCurve {keys [<at 0.0783 0.7758>]}
       RedCurve {keys [<at 0.0783 0.3278>]}
      }})
    Particles.PurpleColor:LinearColor=&ColorFromCurve(ColorCurve=DataInterface() {propertyValues {
       AlphaCurve {keys [<at 0.2168 0>,<at 0.2723 1>,<at 0.6583 1>,<at 0.6788 0>]}
       BlueCurve {keys [<at 0.0684 0.9047>]}
       GreenCurve {keys [<at 0.0684 0.6514>]}
       RedCurve {keys [<at 0.0684 0.0331>]}
      }})
   ) {isSetParameters=true, module=SetVariables_#3}
  ]
  rendererClasses [&NiagaraSpriteRendererProperties]
  renderers [
   {rendererClass=&NiagaraSpriteRendererProperties}
   {rendererClass=&NiagaraSpriteRendererProperties, rendererIndex=1}
  ]
 }
}
summary {
 systemName=NS_GlitchyBit
 emitters [
  {emitterName=NE_GlitchyBlueOutline, rendererClasses [&NiagaraRibbonRendererProperties]}
  {emitterName=NE_HolesParticles, rendererClasses [&NiagaraSpriteRendererProperties]}
  {emitterName=NE_BegginingParticles, rendererClasses [&NiagaraSpriteRendererProperties]}
  {emitterName=NE_Final_Particles, rendererClasses [&NiagaraSpriteRendererProperties]}
  {emitterName=NE_GlitchyGreen, rendererClasses [&NiagaraRibbonRendererProperties]}
 ]
 userVariables [
  User.bIsFrontendPreview {type=NiagaraBool}
  User.bIsPainted {type=NiagaraBool}
  User.ColorVariantID {type=NiagaraInt32}
  User.DTCarOffset {type=Vector3f, description="", defaultValue {struct=Vector3f, value=<0 0 0>}}
  User.GameModeID {type=NiagaraInt32}
  User.PaintedColor {type=LinearColor, description="", defaultValue {struct=LinearColor, value=<rgba 1 1 1 1>}}
  User.PaintedScalar {type=NiagaraFloat, defaultValue=1, description=""}
  User.PortCount {type=NiagaraInt32, defaultValue=2}
 ]
}
